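; Template PCG generator graph for Biome Core that scatters surface points inside a biome's volume.
; A parameter selects World Ray Hit (default) or landscape data as the Surface Sampler's surface, with exposed density, extents, looseness and seed.
; Input tagged BoundingShape bounds the sampler, avoiding costly generation outside the biome.
; An optional branch, on by default, resets X/Y rotation through absolute Transform Points with random yaw before output.
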(graph TPL_BiomeCore_Generator
  (params (SampleOnLandscape :type Bool) (PointsPerSquaredMeter :type Double :default 0.1) (PointExtents :type Vector :default (5e+01 5e+01 5e+01)) (PointsLooseness :type Double :default 1.0) (Seed :type Int :default 1401510044) (Reset_Rotation_XY :type Bool :default true))
  ;[0 0 0 0.5 12] Select between World Ray Hit surface (default) or the Landscape
  (bind GetLandscapeData_0 (Get%20Landscape%20Data :SamplingProperties (struct :bGetLayerWeights false)))
  (bind WorldRayHitQuery_4 (World%20Ray%20Hit%20Query))
  (bind Select_3 (Select :bUseInputB-param SampleOnLandscape))
  ;[1 0 0] Important For Performance : always use the incoming Biome Bounding Shape to avoid generating points outside the biome's volume
  (bind FilterDataByTag_0 (Filter%20Data%20By%20Tag :bTokenizeOnWhiteSpace true :SelectedTags "BoundingShape"))
  ;[0 0 0 1 12] i.e sample world or landscape with a parameterized surface sampler followed by the option to reset sampled points XY rotations
  (bind SurfaceSampler_0 (Surface%20Sampler :Looseness-param PointsLooseness :PointExtents-param PointExtents :PointsPerSquaredMeter-param PointsPerSquaredMeter :Seed-param Seed))
  (bind Branch_2 (Branch :bOutputToB-param Reset_Rotation_XY))
  (bind TransformPoints_1 (Transform%20Points :bAbsoluteRotation true :RotationMax (struct :Yaw 3.6e+02)))
  (connect SurfaceSampler_0 Out Branch_2 In)
  (connect DefaultInputNode In FilterDataByTag_0 In)
  (connect WorldRayHitQuery_4 Out Select_3 Input%20A)
  (connect GetLandscapeData_0 Out Select_3 Input%20B)
  (connect FilterDataByTag_0 InsideFilter SurfaceSampler_0 Bounding%20Shape)
  (connect Select_3 Out SurfaceSampler_0 Surface)
  (connect Branch_2 Output%20B TransformPoints_1 In)
  (output Out Branch_2.Output%20A)
  (output Out TransformPoints_1))
