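; Biome Core advanced generator template: a PCG graph that scatters points on landscape or ray-hit surfaces within a tagged bounding shape.
; Optional slope masking and fractal noise, with optional Voronoi breakup, shape density, which drives scaling via subgraph.
; Spline-distance gradient (EdgeDensity) scales randomly transformed points from edge to center before reprojection and optional weightmap filtering.
; Debug toggles divert stages to Debug nodes; the weight-exclusion parameters are unused here.
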(graph TPL_BiomeCore_Generator_Advanced
  (params (BasicNoiseSeed :type Double :default -12251366.0) (ScaleMin :type Vector :default (0.8 0.8 0.8)) (ScaleMax :type Vector :default (1.2 1.2 1.2)) (RotationMin :type Rotator) (RotationMax :type Rotator :default (struct :Yaw 3.6e+02)) (PointLooseness :type Double :default 1.0) (DebugFinalOutput :type Bool) (EnableSpatialNoise :type Bool :default true) (SpatialDensityMultiplier :type Double :default 1.0) (SpatialDensityLow :type Double :default 0.2) (SpatialDensityHigh :type Double :default 1.0) (SpatialNoiseContrast :type Double :default 1.6) (SpatialNoiseTransform :type Transform :default (struct :Scale3D (25.0 25.0 25.0))) (EnableVoronoiNoiseBreakup :type Bool) (DebugSpatialNoise :type Bool) (VoronoiPower :type Double :default 0.25) (VoronoiTransform :type Transform :default (struct :Scale3D (7e+01 7e+01 7e+01))) (ScaleByDensityRangeMin :type Double :default 1.0) (ScaleByDensityRangeMax :type Double :default 1.0) (DebugVoronoi :type Bool) (EnableSlopeMasking :type Bool :default true) (SlopeMaskLow :type Double :default 0.8) (SlopeMaskHigh :type Double :default 1.0) (DebugSlopeMasking :type Bool) (EdgeDistance :type Double :default 2e+03) (EnableDistanceToEdge :type Bool :default true) (VolumeFalloffScaleEdge :type Double :default 1.0) (VolumeFalloffScaleCenter :type Double :default 1.0) (DebugEdgeGradient :type Bool) (PointsPerSquareMeter :type Double :default 0.1) (EnableExcludePointsByWeight :type Bool :default true) (WeightAttributesToExclude :type String) (ScramblePositionDistance :type Double :default 1e+02) (PointExtents :type Vector :default (2e+01 2e+01 2e+01)) (EnableLandscapeWeightmapLayerFilter :type Bool) (WeightmapLayerDataAsset :type (SoftObject /PCGBiomeCore/BiomeGenerators/Setup/PDA_LandscapeWeightmap_Layer.PDA_LandscapeWeightmap_Layer_C) :default /PCGBiomeCore/BiomeGenerators/Setup/DA_WeightmapLayer_Template.DA_WeightmapLayer_Template) (EnableWorldRayHitQuery :type Bool))
  (bind FilterDataByTag_1 (Filter%20Data%20By%20Tag :bTokenizeOnWhiteSpace true :SelectedTags "BoundingShape"))
  (bind Select_1 (Select :bUseInputB-param EnableWorldRayHitQuery)) ; Use complex trace for staticmesh based landscapes
  (bind SurfaceSampler_0 (Surface%20Sampler :Looseness 0.0 :PointExtents (2e+01 2e+01 2e+01) :PointsPerSquaredMeter 5.0 :PointSteepness 0.5 :Looseness-param PointLooseness :PointExtents-param PointExtents :PointsPerSquaredMeter-param PointsPerSquareMeter))
  (bind Branch_3 (Branch :bOutputToB-param EnableSlopeMasking))
  (bind Branch_6 (Branch :bOutputToB-param EnableSpatialNoise))
  (bind Select_5 (Select :bUseInputB-param EnableSpatialNoise))
  (bind Subgraph_1 (Subgraph :Graph /PCGBiomeCore/GraphTemplates/Graphs/BiomeCore_ScaleByDensity.BiomeCore_ScaleByDensity :ScaleMax ScaleByDensityRangeMax :ScaleMin ScaleByDensityRangeMin))
  (bind Branch_4 (Branch :bOutputToB-param EnableDistanceToEdge))
  (bind Branch_9 (Branch :bOutputToB-param EnableDistanceToEdge))
  (bind Select_2 (Select :bUseInputB-param EnableDistanceToEdge))
  (bind Projection_0 (Projection :ProjectionParams (struct :bProjectRotations false)))
  (bind Branch_12 (Branch :bOutputToB-param EnableLandscapeWeightmapLayerFilter))
  (bind Branch_0 (Branch :bOutputToB-param DebugFinalOutput))
  (bind Debug_0 (Debug))
  ;[0 0 0 1 14] Landscape and Staticmesh
  (bind GetLandscapeData_0 (Get%20Landscape%20Data :bComponentsMustOverlapSelf false :SamplingProperties (struct)))
  (bind WorldRayHitQuery_0 (World%20Ray%20Hit%20Query :QueryParams (struct :bOverrideDefaultParams true :RayOrigin (0.0 0.0 5e+06) :RayLength 5e+06 :bTraceComplex true)))
  ;[0 0 0] Slope Mask
  (bind NormalToDensity_0 (Normal%20To%20Density))
  (bind DensityFilter_2 (Density%20Filter :LowerBound-param SlopeMaskLow :UpperBound-param SlopeMaskHigh))
  (bind Branch_2 (Branch :bOutputToB-param DebugSlopeMasking))
  (bind Debug_2 (Debug))
  ;[0 0 0] Spatial Noise
  (bind Spatial%20Noise_0 (Spatial%20Noise :Brightness 0.1 :Contrast 1.6 :Mode FractionalBrownian2D :Transform (struct :Scale3D (25.0 25.0 25.0)) :Contrast-param SpatialNoiseContrast :Transform-param SpatialNoiseTransform))
  (bind Branch_7 (Branch :bOutputToB-param EnableVoronoiNoiseBreakup))
  (bind Select_6 (Select :bUseInputB-param EnableVoronoiNoiseBreakup))
  (bind DensityFilter_0 (Density%20Filter :LowerBound 0.2 :LowerBound-param SpatialDensityLow :UpperBound-param SpatialDensityHigh))
  (bind AttributeReduce_0 (Attribute%20Reduce :InputSource $Density :Operation Min))
  (bind AttributeReduce_1 (Attribute%20Reduce :InputSource $Density :Operation Max))
  (bind AttributeRemap_1 (Attribute%20Remap :InputSource $Density))
  (bind Branch_1 (Branch :bOutputToB-param DebugSpatialNoise))
  (bind Debug_1 (Debug))
  ;;[0 0 0] Voronoi Breakup
  (bind Spatial%20Noise_1 (Spatial%20Noise :Brightness 0.1 :Contrast 3.0 :Mode Voronoi2D :Transform (struct :Scale3D (7e+01 7e+01 7e+01)) :VoronoiCellIDTarget $Density :Transform-param VoronoiTransform))
  (bind AttributeReduce_2 (Attribute%20Reduce :InputSource $Density :Operation Min))
  (bind AttributeReduce_3 (Attribute%20Reduce :InputSource $Density :Operation Max))
  (bind AttributeRemap_2 (Attribute%20Remap :InputSource $Density))
  (bind AttributeMathsOp_6 (Attribute%20Maths%20Op :InputSource1 $Density :Operation Pow :OutputTarget $Density :InB VoronoiPower))
  (bind AttributeMathsOp_5 (Attribute%20Maths%20Op :InputSource1 $Density :Operation Clamp :OutputTarget $Density))
  (bind AttributeReduce_4 (Attribute%20Reduce :InputSource $Density :Operation Min))
  (bind AttributeReduce_5 (Attribute%20Reduce :InputSource $Density :Operation Max))
  (bind AttributeRemap_3 (Attribute%20Remap :InputSource $Density))
  (bind AttributeMathsOp_1 (Attribute%20Maths%20Op :InputSource1 $Density :InputSource2 $Density :Operation Multiply :OutputTarget $Density))
  (bind AttributeMathsOp_2 (Attribute%20Maths%20Op :Operation Clamp))
  (bind AttributeMathsOp_3 (Attribute%20Maths%20Op :Operation Multiply :InB SpatialDensityMultiplier))
  (bind AttributeMathsOp_4 (Attribute%20Maths%20Op :Operation Clamp))
  ;;;[0.1981 0.01938 0.1981] Debug Voronoi Output
  (bind Branch_8 (Branch :bOutputToB-param DebugVoronoi))
  (bind BoundsModifier_1 (Bounds%20Modifier :BoundsMin (1e+02 1e+02 1e+02) :Mode Set))
  (bind Debug_6 (Debug))
  ;[0 0 0] Scale and Transforms
  (bind MakeVectorAttribute_0 (Make%20Vector%20Attribute :OutputType Vector :X ScramblePositionDistance :Y ScramblePositionDistance))
  (bind AttributeMathsOp_19 (Attribute%20Maths%20Op :Operation Negate))
  (bind TransformPoints_0 (Transform%20Points :RotationMax (struct :Yaw 3.6e+02) :ScaleMax (1.2 1.2 1.2) :ScaleMin (0.8 0.8 0.8) :RotationMax-param RotationMax :RotationMin-param RotationMin :ScaleMax-param ScaleMax :ScaleMin-param ScaleMin))
  ;[0 0 0] Calculate Spline Distance To Center and Set Density
  (bind GetActorData_2 (Get%20Actor%20Data :ComponentSelector (struct :ComponentSelection ByClass :ComponentSelectionClass /Script/Engine.SplineComponent)))
  (bind FilterDataByType_1 (Filter%20Data%20By%20Type :TargetType (struct :Ids [(struct :Struct /Script/PCG.PCGDataTypeInfoPolyline)])))
  (bind SplineSampler_0 (Spline%20Sampler :SamplerParams (struct :Mode Distance :PointSteepness 1.0))) ; 100cm Sampling
  (bind Distance_0 (Distance :bOutputToAttribute false :bSetDensity true :MaximumDistance 2e+03 :OutputAttribute Distance :TargetShape Center :MaximumDistance-param EdgeDistance))
  (bind AttributeRemap_5 (Attribute%20Remap :InputSource $Density :OutputTarget EdgeDensity))
  (bind Branch_10 (Branch :bOutputToB-param DebugEdgeGradient))
  (bind CopyAttributes_2 (Copy%20Attributes :InputSource EdgeDensity :OutputTarget $Density))
  (bind Debug_8 (Debug))
  ;[0 0 0] Scale By Density Based on Spline Distance to Center
  (bind CopyAttributes_0 (Copy%20Attributes :InputSource EdgeDensity :OutputTarget $Density))
  (bind AttributeMathsOp_23 (Attribute%20Maths%20Op :Operation Lerp :OutputTarget EdgeDensity :InA VolumeFalloffScaleEdge :InB VolumeFalloffScaleCenter))
  (bind AttributeMathsOp_9 (Attribute%20Maths%20Op :InputSource1 $Scale :InputSource2 EdgeDensity :Operation Multiply :OutputTarget $Scale)) ; Multiply by $Scale
  (bind AttributeMathsOp_10 (Attribute%20Maths%20Op :InputSource1 $Density :Operation Clamp :OutputTarget $Density))
  ;[0 0 0] Weightmap Layer Filtering
  (bind Subgraph_0 (Subgraph :Graph /PCGBiomeCore/GraphTemplates/Graphs/BiomeCore_WeightmapFilter.BiomeCore_WeightmapFilter :WeightmapLayerDataAsset WeightmapLayerDataAsset))
  (connect AttributeRemap_3 Out AttributeMathsOp_1 InA)
  (connect AttributeMathsOp_5 Out AttributeMathsOp_1 InB)
  (connect AttributeMathsOp_1 Out AttributeMathsOp_2 In)
  (connect AttributeMathsOp_2 Out AttributeMathsOp_3 InA)
  (connect AttributeMathsOp_3 Out AttributeMathsOp_4 In)
  (connect AttributeMathsOp_6 Out AttributeMathsOp_5 In)
  (connect AttributeRemap_2 Out AttributeMathsOp_6 InA)
  (connect CopyAttributes_0 Out AttributeMathsOp_9 InA)
  (connect AttributeMathsOp_23 Out AttributeMathsOp_9 InB)
  (connect AttributeMathsOp_9 Out AttributeMathsOp_10 In)
  (connect MakeVectorAttribute_0 Out AttributeMathsOp_19 In)
  (connect CopyAttributes_0 Out AttributeMathsOp_23 Ratio)
  (connect DensityFilter_0 Out AttributeReduce_0 In)
  (connect DensityFilter_0 Out AttributeReduce_1 In)
  (connect Spatial%20Noise_1 Out AttributeReduce_2 In)
  (connect Spatial%20Noise_1 Out AttributeReduce_3 In)
  (connect Branch_7 Output%20B AttributeReduce_4 In)
  (connect Branch_7 Output%20B AttributeReduce_5 In)
  (connect DensityFilter_0 Out AttributeRemap_1 In)
  (connect AttributeReduce_1 Out AttributeRemap_1 InRangeMax)
  (connect AttributeReduce_0 Out AttributeRemap_1 InRangeMin)
  (connect Spatial%20Noise_1 Out AttributeRemap_2 In)
  (connect AttributeReduce_3 Out AttributeRemap_2 InRangeMax)
  (connect AttributeReduce_2 Out AttributeRemap_2 InRangeMin)
  (connect Branch_7 Output%20B AttributeRemap_3 In)
  (connect AttributeReduce_5 Out AttributeRemap_3 InRangeMax)
  (connect AttributeReduce_4 Out AttributeRemap_3 InRangeMin)
  (connect Distance_0 Out AttributeRemap_5 In)
  (connect Branch_8 Output%20B BoundsModifier_1 In)
  (connect Branch_12 Output%20A Branch_0 In)
  (connect Subgraph_0 Out Branch_0 In)
  (connect AttributeRemap_1 Out Branch_1 In)
  (connect DensityFilter_2 Out Branch_2 In)
  (connect SurfaceSampler_0 Out Branch_3 In)
  (connect Subgraph_1 Points Branch_4 In)
  (connect Branch_2 Output%20A Branch_6 In)
  (connect Branch_3 Output%20A Branch_6 In)
  (connect Spatial%20Noise_0 Out Branch_7 In)
  (connect Spatial%20Noise_1 Out Branch_8 In)
  (connect TransformPoints_0 Out Branch_9 In)
  (connect AttributeRemap_5 Out Branch_10 In)
  (connect Projection_0 Out Branch_12 In)
  (connect Branch_9 Output%20B CopyAttributes_0 Source)
  (connect Branch_9 Output%20B CopyAttributes_0 Target)
  (connect Branch_10 Output%20B CopyAttributes_2 Source)
  (connect Branch_10 Output%20B CopyAttributes_2 Target)
  (connect Branch_0 Output%20B Debug_0 In)
  (connect Branch_1 Output%20B Debug_1 In)
  (connect Branch_2 Output%20B Debug_2 In)
  (connect BoundsModifier_1 Out Debug_6 In)
  (connect CopyAttributes_2 Out Debug_8 In)
  (connect Select_6 Out DensityFilter_0 In)
  (connect NormalToDensity_0 Out DensityFilter_2 In)
  (connect Branch_4 Output%20B Distance_0 Source)
  (connect SplineSampler_0 Out Distance_0 Target)
  (connect DefaultInputNode In FilterDataByTag_1 In)
  (connect GetActorData_2 Out FilterDataByType_1 In)
  (connect Branch_3 Output%20B NormalToDensity_0 In)
  (connect Select_2 Out Projection_0 In)
  (connect Select_1 Out Projection_0 Projection%20Target)
  (connect GetLandscapeData_0 Out Select_1 Input%20A)
  (connect WorldRayHitQuery_0 Out Select_1 Input%20B)
  (connect Branch_9 Output%20A Select_2 Input%20A)
  (connect AttributeMathsOp_10 Out Select_2 Input%20B)
  (connect Branch_6 Output%20A Select_5 Input%20A)
  (connect Branch_1 Output%20A Select_5 Input%20B)
  (connect Branch_7 Output%20A Select_6 Input%20A)
  (connect AttributeMathsOp_4 Out Select_6 Input%20B)
  (connect Branch_6 Output%20B Spatial%20Noise_0 In)
  (connect Branch_7 Output%20B Spatial%20Noise_1 In)
  (connect FilterDataByType_1 InsideFilter SplineSampler_0 Spline)
  (connect Branch_12 Output%20B Subgraph_0 In)
  (connect Select_5 Out Subgraph_1 Points)
  (connect FilterDataByTag_1 InsideFilter SurfaceSampler_0 Bounding%20Shape)
  (connect Select_1 Out SurfaceSampler_0 Surface)
  (connect Branch_4 Output%20A TransformPoints_0 In)
  (connect Branch_10 Output%20A TransformPoints_0 In)
  (connect MakeVectorAttribute_0 Out TransformPoints_0 OffsetMax)
  (connect AttributeMathsOp_19 Out TransformPoints_0 OffsetMin)
  (output Out Branch_0.Output%20A))
